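# Material function for the TungStan character that computes UVs for placing one positioned, scaled, rotatable triangle with an aperture-style slide.
# It rotates the texture coordinates around a slightly offset Position, scales them inversely by TriangleScale and recenters them on that position.
# A small vertical offset fades out as ApertureAmount begins to rise.
# ApertureAmount also shifts the result horizontally, with ApertureRotationDirection choosing left or right.
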MaterialFunction /BRCosmetics/Characters/Player/Male/Medium/Bodies/M_MED_TungStan/Materials/MaterialFunctions/MF_TungStan_TriangleUV
graph {nodes [
  n1 = FunctionInput() {inputName=Position, inputType=FunctionInput_Vector2}
  n2 = FunctionInput() {bUsePreviewValueAsDefault=true, inputName=TriangleScale, inputType=FunctionInput_Scalar, previewValue=<1 0 0 1>}
  n3 = Divide(B=@n2) {constA=10}
  n4 = Multiply(A=@n1, B=@n3)
  Add(A=@n4, B=Constant2Vector())
  n5 = Constant2Vector() {r=1}
  n6 = FunctionInput() {inputName=ApertureAmount, inputType=FunctionInput_Scalar}
  n7 = LinearInterpolate(Alpha=Saturate(None=InverseLinearInterpolate(Value=@n6) {constB=0.07})) {constA=0.0055, constB=0}
  AppendVector(A=@n5.R, B=@n7)
  n8 = TextureCoordinate()
  n9 = Add(A=@n1, B=Constant2Vector() {g=0.0037})
  n10 = FunctionInput() {inputName=RotationAngle, inputType=FunctionInput_Scalar, sortPriority=1}
  n11 = Constant2Vector() {g=0.005}
  n12 = Multiply(
   A=Add(
    A=MaterialFunctionCall(UVs=@n8, Rotation Center=@n9, "Rotation Angle (0-1)"=@n10) {materialFunction=&CustomRotator}.Rotated Values
    B=AppendVector(A=@n11.R, B=@n7)
   )
   B=@n3
  ) {constB=2}
  n13 = Add(A=Multiply(A=@n4) {constB=-1}) {constB=0.5}
  n14 = FunctionInput() {
   description="Value should be either 1 (left) or -1 (right)"
   inputName=ApertureRotationDirection
   inputType=FunctionInput_Scalar
   sortPriority=1
   previewValue=<1 0 0 1>
  }
  n15 = Multiply(A=@n5, B=@n14)
  FunctionOutput(None=Add(A=Add(A=@n12, B=@n13) {constB=-0.5}, B=Multiply(A=@n6, B=@n15))) {outputName=UV}
 ]}
