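# Material function for the TungStan character that builds a six-sided aperture mask from six rotated copies of a triangle mask texture.
# Each copy gets UVs from MF_TungStan_TriangleUV, using shared center, scale, aperture amount and direction, with LocalRotationAmount offset in one-sixth-turn steps and wrapped by Frac.
# The clamped grayscale samples are sharpened with SmoothStep, summed and saturated into the single output.
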MaterialFunction /BRCosmetics/Characters/Player/Male/Medium/Bodies/M_MED_TungStan/Materials/MaterialFunctions/MF_TungStan_6SideAperture_Inverted
graph {nodes [
  n1 = FunctionInput() {bUsePreviewValueAsDefault=true, inputName=LocalRotationAmount, inputType=FunctionInput_Scalar}
  n2 = Add(A=Constant(), B=@n1)
  n3 = FunctionInput() {bUsePreviewValueAsDefault=true, inputName=CenterPosition, inputType=FunctionInput_Vector2}
  n4 = FunctionInput() {
   bUsePreviewValueAsDefault=true
   description="Value should be either 1 (left) or -1 (right)"
   inputName=ApertureRotationDirection
   inputType=FunctionInput_Scalar
   previewValue=<1 0 0 1>
  }
  n5 = FunctionInput() {bUsePreviewValueAsDefault=true, inputName=TriangleScale, inputType=FunctionInput_Scalar, previewValue=<1 0 0 1>}
  n6 = FunctionInput() {bUsePreviewValueAsDefault=true, inputName=ApertureAmount, inputType=FunctionInput_Scalar}
  n7 = TextureSample(UVs=MaterialFunctionCall!(TriangleScale=@n5, ApertureAmount=@n6, Position=@n3, ApertureRotationDirection=@n4, RotationAngle=Frac(None=@n2)) {materialFunction=&/MF_TungStan_TriangleUV}.UV) {
   samplerSource=SSM_Clamp_WorldGroupSettings
   samplerType=SAMPLERTYPE_Grayscale
   texture=/Game/Characters/Player/Male/Medium/Bodies/M_MED_TungStan/Textures/Tungstan_TriangleMask02
  }
  n8 = Add(A=Constant() {r=0.1667}, B=@n1)
  n9 = TextureSample(UVs=MaterialFunctionCall(TriangleScale=@n5, ApertureAmount=@n6, Position=@n3, ApertureRotationDirection=@n4, RotationAngle=Frac(None=@n8)) {materialFunction=&MF_TungStan_TriangleUV}.UV) {
   samplerSource=SSM_Clamp_WorldGroupSettings
   samplerType=SAMPLERTYPE_Grayscale
   texture=&Tungstan_TriangleMask02
  }
  n10 = Add(A=SmoothStep(Value=@n7.RGB) {constMin=0.8}, B=SmoothStep(Value=@n9.RGB) {constMin=0.8})
  n11 = Add(A=Constant() {r=0.3334}, B=@n1)
  n12 = TextureSample(UVs=MaterialFunctionCall(TriangleScale=@n5, ApertureAmount=@n6, Position=@n3, ApertureRotationDirection=@n4, RotationAngle=Frac(None=@n11)) {materialFunction=&MF_TungStan_TriangleUV}.UV) {
   samplerSource=SSM_Clamp_WorldGroupSettings
   samplerType=SAMPLERTYPE_Grayscale
   texture=&Tungstan_TriangleMask02
  }
  n13 = SmoothStep(Value=@n12.RGB) {constMin=0.8}
  n14 = Add(A=Constant() {r=0.5001}, B=@n1)
  n15 = TextureSample(UVs=MaterialFunctionCall(TriangleScale=@n5, ApertureAmount=@n6, Position=@n3, ApertureRotationDirection=@n4, RotationAngle=Frac(None=@n14)) {materialFunction=&MF_TungStan_TriangleUV}.UV) {
   samplerSource=SSM_Clamp_WorldGroupSettings
   samplerType=SAMPLERTYPE_Grayscale
   texture=&Tungstan_TriangleMask02
  }
  n16 = Add(A=Constant() {r=0.6668}, B=@n1)
  n17 = TextureSample(UVs=MaterialFunctionCall(TriangleScale=@n5, ApertureAmount=@n6, Position=@n3, ApertureRotationDirection=@n4, RotationAngle=Frac(None=@n16)) {materialFunction=&MF_TungStan_TriangleUV}.UV) {
   samplerSource=SSM_Clamp_WorldGroupSettings
   samplerType=SAMPLERTYPE_Grayscale
   texture=&Tungstan_TriangleMask02
  }
  n18 = Add(A=Add(A=Add(A=@n10, B=@n13), B=SmoothStep(Value=@n15.RGB) {constMin=0.8}), B=SmoothStep(Value=@n17.RGB) {constMin=0.8})
  n19 = Add(A=Constant() {r=0.8335}, B=@n1)
  n20 = TextureSample(UVs=MaterialFunctionCall(TriangleScale=@n5, ApertureAmount=@n6, Position=@n3, ApertureRotationDirection=@n4, RotationAngle=Frac(None=@n19)) {materialFunction=&MF_TungStan_TriangleUV}.UV) {
   samplerSource=SSM_Clamp_WorldGroupSettings
   samplerType=SAMPLERTYPE_Grayscale
   texture=&Tungstan_TriangleMask02
  }
  n21 = SmoothStep(Value=@n20.RGB) {constMin=0.8}
  FunctionOutput(None=Saturate(None=Add(A=@n18, B=@n21)))
 ]}
