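# MF_TungStan_CircleMaskGenerator is a material function for the TungStan character body that outputs a procedural circular mask in UV space.
# It recenters the UVs and the CenterPosition input around 0.5, scales both inversely by CircleScale, and measures the distance between them.
# A SmoothStep from CircleHardness to 1 shapes the edge, then the result is inverted and saturated so the circle's interior is white.
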MaterialFunction /BRCosmetics/Characters/Player/Male/Medium/Bodies/M_MED_TungStan/Materials/MaterialFunctions/MF_TungStan_CircleMaskGenerator
graph {nodes [
  n1 = Subtract(A=TextureCoordinate()) {constB=0.5}
  n2 = Divide(B=FunctionInput() {inputName=CircleScale, inputType=FunctionInput_Scalar}) {constA=10}
  n3 = Add(A=FunctionInput() {inputName=CenterPosition, inputType=FunctionInput_Vector2}) {constB=-0.5}
  n4 = SmoothStep(
   Min=FunctionInput() {inputName=CircleHardness, inputType=FunctionInput_Scalar, previewValue=<0.97 0 0 1>}
   Max=Constant() {r=1}
   Value=Distance(A=Multiply(A=@n1, B=@n2) {constB=5}, B=Multiply(A=@n2, B=@n3))
  )
  FunctionOutput(None=Saturate(None=OneMinus(None=@n4))) {outputName=Mask}
 ]}
